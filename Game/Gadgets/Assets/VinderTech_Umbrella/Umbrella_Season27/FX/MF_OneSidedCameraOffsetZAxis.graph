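# Material function in the Season 27 umbrella FX folder that outputs a one-sided, camera-dependent offset along the Z axis.
# It dots CameraDirectionVector with the up axis and clamps the result, so the offset fades with camera pitch and is zero when that dot product is negative.
# The clamped value scales the OffsetAmount input, which falls back to its preview value, and multiplies a Z unit vector.
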MaterialFunction /Game/Gadgets/Assets/VinderTech_Umbrella/Umbrella_Season27/FX/MF_OneSidedCameraOffsetZAxis
graph {nodes [
  n1 = Clamp(None=DotProduct(
    A=MaterialFunctionCall() {materialFunction=&CameraDirectionVector}.Result
    B=Constant3Vector() {constant=<rgba 0 0 1 0>}
   ))
  n2 = FunctionInput() {bUsePreviewValueAsDefault=true, inputName="OffsetAmount(S)", inputType=FunctionInput_Scalar, previewValue=<15 0 0 1>}
  FunctionOutput(None=Multiply(A=Multiply(A=@n1, B=@n2), B=Constant3Vector() {constant=<rgba 0 0 1 0>})) {outputName=Multiply}
 ]}
