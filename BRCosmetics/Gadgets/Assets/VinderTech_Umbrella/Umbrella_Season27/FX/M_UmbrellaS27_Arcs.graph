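# Translucent unlit material for electric arc effects on the Season 27 umbrella glider, used on Niagara mesh particles.
# SplineThicken builds the arc width from panning noise, tapered tips and dynamic parameters, adding a camera Z offset.
# Opacity multiplies particle alpha, edge gradient, optional moving zap mask and panning lightning textures, selected by glow parameter.
# A bypassable sphere mask cuts arcs inside the umbrella radius, with depth fade on deferred shading.
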Material /BRCosmetics/Gadgets/Assets/VinderTech_Umbrella/Umbrella_Season27/FX/M_UmbrellaS27_Arcs
graph {
 properties {
  bUsedWithNiagaraMeshParticles=true
  blendMode=BLEND_Translucent
  pixelDepthOffsetMode=PDOM_Legacy
  shadingModel=MSM_Unlit
  usedShadingModels=Unlit
 }
 nodes [
  Constant() {r=1}
  MaterialFunctionCall(WidthBase=Constant() {r=50}, WidthTip=Constant() {r=50}) {materialFunction=&SplineThicken}
  n1 = ParticleColor()
  MaterialFunctionCall(Emissive=@n1.RGB) {materialFunction=&MF_FX_EmissiveAdjustments}
  n2 = Multiply(
   A=@n1.A
   B=MaterialFunctionCall!(UVs, Offset, RotationCenter, RotationAngle=Constant() {r=0.25}) {materialFunction=/FortUILibrary/MaterialFunctions/Gradients/MF_UI_ReflectedGradient}.Result
  )
  n3 = ComponentMask(None=TextureCoordinate()) {r=true}
  n4 = LinearInterpolate(A=@n3, B=OneMinus(None=@n3), Alpha=Round(None=ParticleRandom()))
  n5 = ScalarParameter() {defaultValue=1, parameterName=UseParticleRandomFlip}
  n6 = Frac(None=Time())
  n7 = DynamicParameter() {defaultValue=<rgba 0.5 1 1 1.2>, paramNames [BaseWidth, Mask Speed, Mask Size, LightningTileX]}
  n8 = Ceil(None=ParticleRelativeTime() {desc=">0"})
  n9 = Constant() {r=3}
  n10 = LinearInterpolate(
   B=MaterialFunctionCall!(
    Gradient=LinearInterpolate(A=@n3, B=@n4, Alpha=@n5)
    Mask Position=LinearInterpolate(A=@n6, B=@n7.Mask Size, Alpha=@n8) {constA=0.5}
    Mask Size=LinearInterpolate(A=@n9, B=@n7.Mask Speed, Alpha=@n8)
   ) {materialFunction=/Game/Effects/Fort_Effects/MaterialFunctions/MF_GradientMask}.Linear
   Alpha=ScalarParameter() {parameterName=isZap?}
  ) {constA=1}
  n11 = DynamicParameter() {defaultValue=<rgba 0.5 1 1 1.2>, paramNames [BaseWidth, Mask Speed, Mask Size, LightningTileX]}
  n12 = Constant() {r=1}
  n13 = Panner(Coordinate=Multiply(A=TextureCoordinate(), B=AppendVector(A=@n11.LightningTileX, B=@n12))) {speedX=-0.5}
  n14 = TextureSample(UVs=@n13) {texture=/BRCosmetics/Weapons/FORT_Melee/Pickaxe_AllKnowing/FX/T_Pickaxe_AllKnowingl_Lightning_01}
  n15 = TextureSample(UVs=@n13) {texture=/BRCosmetics/Weapons/FORT_Melee/Pickaxe_AllKnowing/FX/T_Pickaxe_AllKnowingl_Lightning_02}
  n16 = ScalarParameter() {parameterName=isGlow?}
  n17 = SphereMask(
   A=TransformPosition(None=WorldPosition().XYZ) {transformSourceType=TRANSFORMPOSSOURCE_World}
   B=VectorParameter() {defaultValue=<rgba 0 0 -162.5272 0>}.RGB
   Radius=Multiply(A=ScalarParameter() {defaultValue=157.5, parameterName=Radius}) {constB=1.03, desc=weird umbrella size conversion}
   Hardness=ScalarParameter() {defaultValue=1, parameterName=Hardness}
  )
  n18 = Multiply(
   A=Multiply(A=Multiply(A=@n2, B=@n10), B=LinearInterpolate(A=@n14.R, B=@n15.G, Alpha=@n16))
   B=LinearInterpolate(A=OneMinus(None=@n17), Alpha=ScalarParameter() {parameterName=BypassSphereMask})
  )
  n19 = DepthFade(Opacity=@n18, FadeDistance=ScalarParameter() {defaultValue=1, parameterName=FadeDistance})
  Multiply(A=ShadingPathSwitch(Default=@n19, Deferred=@n19, Forward=@n18, Mobile=@n18), B=ScalarParameter() {defaultValue=4, parameterName=Param_1})
  n20 = Constant()
  n21 = Multiply(A=ComponentMask(None=TextureCoordinate()) {r=true}, B=ScalarParameter() {defaultValue=1, parameterName=UVScale})
  n22 = ScalarParameter() {defaultValue=0.3, parameterName=SpeedX}
  n23 = ScalarParameter() {defaultValue=-0.5, parameterName=SpeedY}
  n24 = TextureSample(UVs=Panner(Coordinate=AppendVector(A=@n20, B=@n21), Time=Time(), Speed=AppendVector(A=@n22, B=@n23))) {
   samplerType=SAMPLERTYPE_LinearGrayscale
   texture=/BRCosmetics/Weapons/FORT_Melee/Pickaxe_Parfait_Female/FX/Textures/T_TrifleParfait_LightningNoise
  }
  NamedRerouteDeclaration(None=@n24.R) {name=Noise, nodeColor=<rgba 0.7109 0 1 1>}
  n25 = ScalarParameter() {defaultValue=1.2, parameterName=MinWidth}
  n26 = ScalarParameter() {defaultValue=8, parameterName=AdditionalWidth}
  n27 = ConstantBiasScale(None=NamedRerouteUsage()) {bias=-0.5, scale=2}
  n28 = ScalarParameter() {defaultValue=2, parameterName=Range}
  n29 = Multiply(
   A=Add(A=Add(A=@n25, B=@n26), B=Multiply(A=@n27, B=@n28))
   B=DynamicParameter() {defaultValue=<rgba 0.5 1 1 1.2>, paramNames [BaseWidth, Mask Speed, Mask Size, LightningTileX]}.BaseWidth
  ) {constA=1}
  n30 = MaterialFunctionCall() {materialFunction=&MF_UI_ReflectedGradient}
  n31 = ScalarParameter() {parameterName=Exp_1}
  n32 = LinearInterpolate(
   A=Multiply(A=Multiply(A=@n29) {constB=-1, desc=Flip Normals}, B=Power(Base=@n30.Result, Exp=@n31)) {desc=Taper Tips}
   B=Constant() {r=1}
   Alpha=ScalarParameter() {parameterName=BypassSplineThickenUVGradient}
  )
  n33 = Multiply(A=@n32, B=ScalarParameter() {defaultValue=1, parameterName=BaseWidthMult})
  n34 = Multiply(A=@n32, B=ScalarParameter() {defaultValue=1, parameterName=TipWidthMult})
  n35 = StaticBool()
  n36 = ScalarParameter() {defaultValue=17}
  Reroute(None=Add(
    A=MaterialFunctionCall(WidthBase=@n33, WidthTip=@n34, Expand U or V UV Channel=@n35) {materialFunction=&SplineThicken}.WorldPositionOffset
    B=MaterialFunctionCall!("OffsetAmount(S)"=@n36) {materialFunction=/Game/Gadgets/Assets/VinderTech_Umbrella/Umbrella_Season27/FX/MF_OneSidedCameraOffsetZAxis}.Multiply
   ))
 ]
}
