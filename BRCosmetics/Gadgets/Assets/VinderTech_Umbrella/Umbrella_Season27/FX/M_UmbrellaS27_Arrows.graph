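# Unlit material for arrow effects on the Season 27 umbrella, enabled for Niagara mesh particles.
# The Color parameter, a bright blue by default, is multiplied by vertex color and blended at a fixed default ratio with a muted Color2 parameter.
# The blend feeds the emissive input of the shared MF_FX_EmissiveAdjustments function, with no textures involved.
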Material /BRCosmetics/Gadgets/Assets/VinderTech_Umbrella/Umbrella_Season27/FX/M_UmbrellaS27_Arrows
graph {
 properties {bUsedWithNiagaraMeshParticles=true, pixelDepthOffsetMode=PDOM_Legacy, shadingModel=MSM_Unlit, usedShadingModels=Unlit}
 nodes [
  n1 = VectorParameter() {parameterName=Color, defaultValue=<rgba 0.252 1.4389 4 1>}
  n2 = VertexColor()
  MaterialFunctionCall(Emissive=LinearInterpolate(A=Multiply(A=@n1.RGB, B=@n2), B=VectorParameter() {parameterName=Color2, defaultValue=<rgba 0.2596 0.365 0.4635 1>}.RGB)) {materialFunction=&MF_FX_EmissiveAdjustments}
 ]
}
