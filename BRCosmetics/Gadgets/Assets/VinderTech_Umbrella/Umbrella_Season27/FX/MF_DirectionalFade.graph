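# MF_DirectionalFade is a material function in the Season 27 umbrella FX that outputs a view-direction-based fade scalar.
# It dots the camera direction vector with a vector parameter defaulting to -Z, multiplies by OpacityMultiplier, then smoothsteps between Min and Max parameters into the Dot output.
# The WSVectorComparision input, Power node and debug node do not feed the output.
# Despite its description, the function applies no mesh offset.
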MaterialFunction /BRCosmetics/Gadgets/Assets/VinderTech_Umbrella/Umbrella_Season27/FX/MF_DirectionalFade
graph {
 properties {description=Offsets the mesh in a direction according to a dot product}
 nodes [
  FunctionInput() {bUsePreviewValueAsDefault=true, inputName=WSVectorComparision, previewValue=<-0.3 0 -1 0>}
  n1 = FunctionInput() {bUsePreviewValueAsDefault=true, inputName=OpacityMultiplier, inputType=FunctionInput_Scalar, previewValue=<1 0 0 0>}
  n2 = Multiply(
   A=DotProduct(
    A=MaterialFunctionCall() {materialFunction=&CameraDirectionVector}.Result
    B=VectorParameter() {defaultValue=<rgba 0 0 -1 0>}.RGB
   )
   B=@n1
  )
  n3 = SmoothStep(Min=ScalarParameter() {parameterName=Min}, Max=ScalarParameter() {parameterName=Max}, Value=@n2) {constMax=0.1, constMin=-0.5}
  FunctionOutput(None=@n3) {outputName=Dot}
  MaterialFunctionCall(Number=@n3) {materialFunction=&DebugScalarValues}
  Power(Base=@n2, Exp=ScalarParameter() {defaultValue=1})
 ]
}
