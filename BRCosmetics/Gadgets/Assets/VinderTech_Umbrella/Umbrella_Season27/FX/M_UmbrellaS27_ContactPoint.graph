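# Unlit translucent material for Niagara mesh particles, named as the Season 27 umbrella's contact-point effect, with depth testing disabled.
# Opacity combines sine UV masks, particle alpha, directional fade and a distance fade that brightens near geometry behind it, bypassing the mask and distance fade on forward and mobile paths.
# Panned lightning noise modulates SplineThicken width, added to a one-sided camera Z offset for vertex displacement.
# Particle-tinted HDR color feeds emissive.
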Material /BRCosmetics/Gadgets/Assets/VinderTech_Umbrella/Umbrella_Season27/FX/M_UmbrellaS27_ContactPoint
graph {
 properties {
  bDisableDepthTest=true
  bUsedWithNiagaraMeshParticles=true
  blendMode=BLEND_Translucent
  pixelDepthOffsetMode=PDOM_Legacy
  shadingModel=MSM_Unlit
  usedShadingModels=Unlit
 }
 nodes [
  DynamicParameter() {defaultValue=<rgba 0.5 1 1 1.2>, paramNames [BaseWidth, Mask Speed, Mask Size, LightningTileX]}
  MaterialFunctionCall(Emissive=Multiply(A=VectorParameter() {parameterName=Param_2, defaultValue=<rgba 0 106.9441 400 1>}.RGB, B=ParticleColor().RGB)) {materialFunction=&MF_FX_EmissiveAdjustments}
  n1 = TextureCoordinate()
  n2 = ComponentMask(None=@n1) {g=true}
  n3 = Saturate(None=Sine(None=Divide(A=@n2)))
  n4 = ComponentMask(None=@n1) {r=true}
  n5 = Saturate(None=Sine(None=Divide(A=@n4)))
  n6 = Distance(
   A=WorldPosition() {worldPositionShaderOffset=WPT_ExcludeAllShaderOffsets}.XYZ
   B=MaterialFunctionCall() {materialFunction=&WorldPositionBehindTranslucency}.Result
  )
  n7 = ScalarParameter() {defaultValue=2, parameterName=FadeDistance}
  n8 = OneMinus(None=Saturate(None=Divide(A=@n6, B=@n7)))
  n9 = ScalarParameter() {defaultValue=4, parameterName=Fade Exponent}
  n10 = Multiply(
   A=Multiply(A=Multiply(A=@n3, B=@n5), B=Power(Base=@n8, Exp=@n9))
   B=Power(Base=OneMinus(None=@n5), Exp=ScalarParameter() {defaultValue=2, parameterName=Exp})
  )
  n11 = Constant() {r=1}
  Multiply(
   A=Multiply(A=ShadingPathSwitch(Default=@n10, Deferred=@n10, Forward=@n11, Mobile=@n11), B=ParticleColor().A)
   B=MaterialFunctionCall!(OpacityMultiplier=ScalarParameter() {defaultValue=1}, WSVectorComparision=Constant3Vector() {constant=<rgba 0 0 -1 0>}) {materialFunction=&/MF_DirectionalFade}.Dot
  )
  n12 = Constant()
  n13 = Multiply(A=ComponentMask(None=TextureCoordinate()) {r=true}, B=ScalarParameter() {defaultValue=1, parameterName=UVScale})
  n14 = ScalarParameter() {defaultValue=0.3, parameterName=SpeedX}
  n15 = ScalarParameter() {defaultValue=-0.5, parameterName=SpeedY}
  n16 = TextureSample(UVs=Panner(Coordinate=AppendVector(A=@n12, B=@n13), Time=Time(), Speed=AppendVector(A=@n14, B=@n15))) {
   samplerType=SAMPLERTYPE_LinearGrayscale
   texture=/BRCosmetics/Weapons/FORT_Melee/Pickaxe_Parfait_Female/FX/Textures/T_TrifleParfait_LightningNoise
  }
  NamedRerouteDeclaration(None=@n16.R) {name=Noise, nodeColor=<rgba 0 1 0.2031 1>}
  n17 = NamedRerouteUsage()
  n18 = Add(
   A=Add(A=ScalarParameter() {defaultValue=1.2, parameterName=MinWidth}, B=ScalarParameter() {defaultValue=8, parameterName=AdditionalWidth})
   B=Multiply(A=ConstantBiasScale(None=@n17) {bias=-0.5, scale=2}, B=ScalarParameter() {defaultValue=2, parameterName=Range})
  )
  n19 = ScalarParameter() {defaultValue=4, parameterName=woid}
  n20 = Multiply(A=Multiply(A=Multiply(A=@n18, B=@n19)) {constB=-1, desc=Flip Normals}) {desc=Taper Tips}
  n21 = StaticBool()
  n22 = ScalarParameter() {defaultValue=32, parameterName="OffsetAmount(S) (S)"}
  Reroute(None=Add(
    A=MaterialFunctionCall(WidthBase=@n20, WidthTip=@n20, Expand U or V UV Channel=@n21) {materialFunction=&SplineThicken}.WorldPositionOffset
    B=MaterialFunctionCall!("OffsetAmount(S)"=@n22) {materialFunction=/Game/Gadgets/Assets/VinderTech_Umbrella/Umbrella_Season27/FX/MF_OneSidedCameraOffsetZAxis}.Multiply
   ))
  VectorParameter() {defaultValue=<rgba 400 400 400 400>}
 ]
}
