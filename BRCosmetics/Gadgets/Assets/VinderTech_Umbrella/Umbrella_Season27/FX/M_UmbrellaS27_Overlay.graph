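# Additive unlit overlay glow material for the Season 27 umbrella glider effects, usable on Niagara mesh particles.
# Particle color feeds an emissive-adjustment function, while a center-scaled exponential radial gradient builds a soft mask multiplied by opacity and particle alpha.
# Depth fade softens that mask on non-mobile shading paths, with mobile using the raw gradient.
# A world-position-offset chain combines switchable camera offset with adjustable vertex-normal push.
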Material /BRCosmetics/Gadgets/Assets/VinderTech_Umbrella/Umbrella_Season27/FX/M_UmbrellaS27_Overlay
graph {
 properties {
  bUsedWithNiagaraMeshParticles=true
  blendMode=BLEND_Additive
  pixelDepthOffsetMode=PDOM_Legacy
  shadingModel=MSM_Unlit
  usedShadingModels=Unlit
 }
 nodes [
  Constant() {r=40}
  n1 = ParticleColor()
  MaterialFunctionCall(Emissive=@n1.RGB) {materialFunction=&MF_FX_EmissiveAdjustments}
  n2 = VectorParameter() {parameterName="Texture Scale (V2)", defaultValue=<rgba 0.9354 1.3454 0.5 0.62>}
  n3 = MaterialFunctionCall(
   UVs=MaterialFunctionCall(Texture Scale=@n2.RGB) {materialFunction=&ScaleUVsByCenter}.UVs
   CenterPosition=AppendVector(A=@n2.B, B=@n2.B)
   Radius=ScalarParameter() {defaultValue=0.2, parameterName="Radius (S)"}
   Density=ScalarParameter() {defaultValue=0.2, parameterName="Density (S)"}
  ) {materialFunction=&RadialGradientExponential}
  n4 = DepthFade(Opacity=@n3.RadialGradientExponential, FadeDistance=ScalarParameter() {defaultValue=20, parameterName=FadeDistance}) {fadeDistanceDefault=1}
  Multiply(
   A=Multiply(
    A=ShadingPathSwitch(Default=@n4, Deferred=@n4, Mobile=@n3.RadialGradientExponential) {desc=Migrated from FeatureLevelSwitch}
    B=ScalarParameter() {defaultValue=1, parameterName=Opacity}
   )
   B=@n1.A
  )
  n5 = VertexNormalWS()
  n6 = StaticSwitchParameter(
   True=Add(
    A=MaterialFunctionCall() {materialFunction=/Engine/Functions/Engine_MaterialFunctions02/WorldPositionOffset/CameraOffset}.Camera Offset
    B=Multiply(A=@n5) {constB=2}
   )
   False=Constant()
  ) {defaultValue=true, parameterName=UseCameraOffset}
  n7 = Multiply(A=VertexNormalWS(), B=ScalarParameter() {defaultValue=1, parameterName=VertexNormalAmount})
  Reroute(None=LinearInterpolate(A=@n6, B=Add(A=@n6, B=@n7), Alpha=ScalarParameter() {defaultValue=1, parameterName=UseNormalMult?}))
 ]
}
