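# Niagara idle effect for the Season 27 umbrella glider: a persistent arrow mesh with a decelerating spin, a sprite glow and electric arcs.
# Two lightning emitters, one repeating at random intervals and one bursting at startup, pick arc meshes from a user index array and add glow, core sprites and lights.
# VariantIndex blends color palettes, while system flicker and time-of-day scaling modulate brightness.
# User.Debug adds extra arcs and forces a mesh index override.
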NiagaraSystem /BRCosmetics/Gadgets/Assets/VinderTech_Umbrella/Umbrella_Season27/FX/NS_Season27_Glider_Idle
dependencies {
 usedDataInterfaces [
  &NiagaraDataInterfacePlatformSet
  &NiagaraDataInterfaceCurve
  &NiagaraDataInterfaceArrayFloat3
  &NiagaraDataInterfaceArrayFloat
 ]
 usedDynamicInputs [&ReturnNormalizedExecIndex, &SelectLinearColorFromArray]
 usedModules [
  &self:SystemSpawnScript.SystemScriptSource.SystemScriptGraph.NiagaraNodeAssignment_0.SetVariables_#1
  &SystemState
  &self:SystemSpawnScript.SystemScriptSource.SystemScriptGraph.NiagaraNodeAssignment_1.SetVariables_#2
  &self:SystemSpawnScript.SystemScriptSource.SystemScriptGraph.NiagaraNodeAssignment_3.SetVariables_#3
 ]
}
emitters {
 Arrow {
  emitterUpdateScript [
   &EmitterState(Life Cycle Mode=::NewEnumerator1, Loop Behavior=::NewEnumerator1, Loop Duration Mode=::NewEnumerator1, UseLoopDelay=false)
   &SpawnBurst_Instantaneous(Spawn Count=3, Use Spawn Probability=false, UseLoopCountLimit=false)
   /SetVariables(
    Emitter.DarkColor:LinearColor=&LerpLinearColors(
     StartColor=<rgba 0.2588 0.3647 0.4627 1>
     EndColor=<rgba 0.3268 0.2588 0.4627 1>
     LerpFactor=&MakeFloatFromInt(Int=$User.VariantIndex)
    )
    Emitter.HighColor:LinearColor=&LerpLinearColors(
     StartColor=<rgba 0.252 1.4389 4 1>
     EndColor=<rgba 0.6893 0.252 4 1>
     LerpFactor=&MakeFloatFromInt(Int=$User.VariantIndex)
    )
   ) {isSetParameters=true}
  ]
  particleSpawnScript [
   &PartitionParticles()
   &InitializeParticle(
    Color=$System.PrimaryColor
    Lifetime=2
    Mesh Scale Mode=::NewEnumerator3
    Mesh Uniform Scale=&SelectFloatFromArray(
     Direct Array Index=&ReturnExecIndex()
     Float Selection Array=DataInterface() {propertyValues {FloatData [1.5,2]}}
    )
    Position Offset=&SelectVectorFromArray(
     Direct Array Index=&ReturnExecIndex()
     Vector Selection Array=DataInterface() {propertyValues {FloatData [<0 0 16>, <0 0 16>, <0 0 16>]}}
    )
    Position Offset Coordinate Space=::Local
    Sprite Rotation Mode=::NewEnumerator1
    UsePositionOffset=true
    Write Material Random=true
   )
   &InitialMeshOrientation(
    Mesh Orientation Mode=::NewEnumerator7
    Rotation=&SelectVectorFromArray(
     Direct Array Index=&ReturnExecIndex()
     Vector Selection Array=DataInterface() {propertyValues {FloatData [<0 0 0>, <0 0 0.2>, <0 0 0>]}}
    )
    Use Rotation Vector=true
   )
   /SetVariables(Particles.MeshIndex=&ReturnExecIndex()) {isSetParameters=true}
   &CameraOffset(Camera Offset Amount=2)
  ]
  particleUpdateScript [
   &ParticleState(Kill Particles When Lifetime Has Elapsed=false, Let Infinitely Lived Particles Die When Emitter Deactivates=true)
   &UpdateMeshOrientation(
    Rotation Rate=&FloatFromCurve(FloatCurve=DataInterface() {propertyValues {Curve {keys [<at 0 5>,<at 1.5 0>]}}}, CurveIndex=$Emitter.Age)
    Rotation Vector=<0 0 1>
   ) {module=UpdateMeshOrientation001}
   /SetVariables(
    Particles.CoreColor:LinearColor=&MultiplyLinearColor(
     Linear Color=<rgba 1.28 18.0587 40 0.2>
     Scale Factor=&MakeVector4_FromFloat(Float=&Lerp_Float(A=0.7, B=1, Alpha=$System.Flicker))
    )
    Particles.CoreSize:Vector2f=&Multiply_Vector2D_ByFloat(Float=20, Vector2D=<1 1>)
   ) {isSetParameters=true}
   &ScaleMeshSize(Scale Factor=&VectorFromFloat(Value=&FN_FrontEndFloatModifier(Gameplay=0.75, Front End=0.85)))
   &UpdateMeshOrientation(
    Rotation Rate=&SelectFloatFromArray(
     Direct Array Index=&ReturnExecIndex()
     Float Selection Array=DataInterface() {propertyValues {FloatData [-0.4,-0.8]}}
    )
    Rotation Vector=<0 0 1>
   ) {enabled=false}
   &SolveForcesAndVelocity(Clamp Velocity=false, Limit Acceleration=false)
  ]
  rendererClasses [&NiagaraMeshRendererProperties, &NiagaraSpriteRendererProperties]
  renderers [
   {rendererClass=&NiagaraMeshRendererProperties, rendererIndex=0}
   {rendererClass=&NiagaraSpriteRendererProperties, rendererIndex=1}
  ]
 }
 LongLightningArcs {
  emitterSpawnScript [
   /SetVariables(Emitter.LightningWidth:NiagaraFloat=&Multiply_Float(A=&RandomRangeFloat(Minimum=0.5, Maximum=1), B=$System.LightningWidth)) {isSetParameters=true, module=SetVariables_#4}
   /SetVariables(Emitter.LightningGlow_Width:NiagaraFloat=&Multiply_Float(A=$Emitter.LightningWidth, B=6)) {isSetParameters=true, module=SetVariables_#5}
  ]
  emitterUpdateScript [
   &EmitterState(
    Inactive Response=::NewEnumerator1
    Life Cycle Mode=::NewEnumerator1
    Loop Duration=&RandomRangeFloat(Minimum=0.4, Maximum=0.8)
    Recalculate Duration Each Loop=true
    UseLoopDelay=false
   )
   &SpawnBurst_Instantaneous(
    Spawn Count=&RandomRangeInt(Minimum=1, Maximum=1)
    Use Spawn Probability=false
    UseLoopCountLimit=false
   ) {module=SpawnBurst_Instantaneous001}
   &SpawnRate(
    Spawn Probability=0.6
    SpawnRate=&Lerp_Float(A=&FN_FrontEndFloatModifier(Gameplay=2, Front End=1.5), B=5, Alpha=$User.ForwardDot)
    Use Spawn Probability=true
   ) {enabled=false}
   &SpawnBurst_Instantaneous(
    Spawn Count=&MakeIntFromFloat(Float=&Lerp_Float(A=0, B=20, Alpha=$User.Debug))
    Use Spawn Probability=false
    UseLoopCountLimit=false
   )
  ]
  particleSpawnScript [
   &PartitionParticles(Number Of Partitions=4)
   &InitializeParticle(
    Color=$System.PrimaryColor
    Lifetime=&RandomRangeFloat(Minimum=1.4, Maximum=0.5)
    Mesh Scale Mode=::NewEnumerator3
    Mesh Uniform Scale=$User.InchesConversion
    Position Offset=&MakeCustomVectorFromBool(
     Bool=&IsPlatformSetActive(New Platform Set=DataInterface() {propertyValues {Platforms {qualityLevelMask=-29}}})
     True Vector=<0 0 1.5>
     False Vector=<0 0 0>
    )
    Position Offset Coordinate Space=::Local
    Sprite Size Mode=::NewEnumerator3
    Uniform Sprite Size=200
    UsePositionOffset=true
    Write Material Random=true
   )
   /SetVariables(
    Particles.ArcMeshIndex:NiagaraInt32=&MakeIntFromFloat(Float=&Lerp_Float(
      A=&MakeFloatFromInt(Int=&SelectIntFromArray(
        Direct Array Index=&ModuloInt(A=&ReturnParticleIDAttributes(), B=12)
        Int Selection Array=$User.Int32 Array
       ))
      B=$User.MeshIndexOverride
      Alpha=$User.Debug
     ))
    Particles.ArcsOffset:NiagaraPosition=&AddVectorToPosition(Position=$Engine.Emitter.SimulationPosition, Vector=<0 0 0>)
   ) {isSetParameters=true, module=SetVariables_#6}
   /SetVariables(Particles.OverlayScale:Vector3f=&VectorFromFloat(Value=2.5)) {isSetParameters=true, module=SetVariables_#7}
   &InitialMeshOrientation(
    Mesh Orientation Mode=::NewEnumerator7
    Rotation=&MakeVector(X=0, Y=0, Z=&RandomRangeFloat(Minimum=-0.35, Maximum=0.45))
    Use Rotation Vector=true
   )
   &PartitionParticles() {module=PartitionParticles001}
   &CameraOffset()
  ]
  particleUpdateScript [
   &ParticleState()
   &ScaleMeshSize(Scale Factor=&MakeVector(
     X=1
     Y=&RandomRangeFloat(
      Minimum=&Lerp_Float(A=0.85, B=1, Alpha=$Particles.NormalizedAge)
      Maximum=&Lerp_Float(A=0.6, B=0.8, Alpha=$Particles.NormalizedAge)
     )
     Z=&FloatFromCurve(FloatCurve=DataInterface() {propertyValues {Curve {keys [
          {arriveTangent=0.9312, interpMode=RCIM_Cubic, leaveTangent=0.9312, tangentMode=RCTM_User, time=0, value=0.4}
          {interpMode=RCIM_Cubic, tangentMode=RCTM_User, time=1, value=1}
         ]}}})
    ))
   &SolveForcesAndVelocity(Clamp Velocity=false, Limit Acceleration=false)
   &ScaleColor(
    Scale RGB=&VectorFromFloat(Value=&Lerp_Float(A=0.6, B=2, Alpha=$System.Flicker))
    ScaleA=false
    ScaleRGB=true
   )
   &ScaleColor(
    Color Value To Scale=$Particles.Color
    Scale RGB=&Multiply_VectorByFloat(Vector=<1 1 1>, Float=$System.TimeOfDayBrightnessMultiplier)
    ScaleA=true
    ScaleRGB=true
   ) {module=ScaleColor001}
   &DynamicMaterialParameters(
    Index 0 Param 1=&FloatFromCurve(
     FloatCurve=DataInterface() {propertyValues {Curve {keys [
         {interpMode=RCIM_Cubic, tangentMode=RCTM_User, time=0, value=0}
         {interpMode=RCIM_Cubic, tangentMode=RCTM_User, time=0.5, value=1.05}
         {interpMode=RCIM_Cubic, tangentMode=RCTM_User, time=1, value=0}
        ]}}}
     Scale Curve=&RandomRangeFloat(Minimum=1, Maximum=2)
    )
    Index 0 Param 4=&RandomRangeFloat(Minimum=1.4, Maximum=1.2)
    Module.Index0Param0Write=true
    Module.Index0Param1Write=true
    Module.Index0Param2Write=true
    Module.Index0Param3Write=true
   )
   /SetVariables(
    Particles.ContactColor:LinearColor=&MultiplyLinearColor(
     Linear Color=$Particles.Color
     Scale Factor=&MakeFloat4FromVectorAndFloat(
      "Vector (RGB)"=&VectorFromFloat(Value=&FNTimeOfDayScalar(Night Scale=0.1, Morning Scale=1, Evening Scale=1))
      "Float (Alpha)"=&FloatFromCurve(FloatCurve=DataInterface() {propertyValues {Curve {keys [
           {interpMode=RCIM_Cubic, tangentMode=RCTM_User, time=0.45, value=0}
           {interpMode=RCIM_Cubic, tangentMode=RCTM_User, time=0.5, value=1}
           {interpMode=RCIM_Cubic, tangentMode=RCTM_User, time=1, value=0}
          ]}}})
     )
    )
    Particles.LightningGlowColor:LinearColor=&MultiplyLinearColor(
     Linear Color=&MultiplyLinearColor(
      Linear Color=$Particles.Color
      Scale Factor=&MakeVector4_FromLinearColor(LinearColor=&LerpLinearColors(
        StartColor=<rgba 0 0.9091 4 1>
        EndColor=<rgba 4 0 2.0667 1>
        LerpFactor=&MakeFloatFromInt(Int=$User.VariantIndex)
       ))
     )
     Scale Factor=&MakeFloat4FromVectorAndFloat("Vector (RGB)"=<0.1 0.1 0.1>, "Float (Alpha)"=0.5)
    )
    Particles.OverlayColor:LinearColor=&MultiplyLinearColor(
     Linear Color=&MultiplyLinearColor(
      Linear Color=$Particles.Color
      Scale Factor=&MakeVector4_FromLinearColor(LinearColor=&LerpLinearColors(
        StartColor=<rgba 0 2.2 4 1>
        EndColor=<rgba 40 0 17.3333 1>
        LerpFactor=&MakeFloatFromInt(Int=$User.VariantIndex)
       ))
     )
     Scale Factor=&MakeFloat4FromVectorAndFloat(
      "Vector (RGB)"=<1 1 1>
      "Float (Alpha)"=&FloatFromCurve(
       FloatCurve=DataInterface() {propertyValues {Curve {keys [
           {interpMode=RCIM_Cubic, tangentMode=RCTM_User, time=0, value=0}
           {interpMode=RCIM_Cubic, tangentMode=RCTM_User, time=0.8, value=1}
           {interpMode=RCIM_Cubic, tangentMode=RCTM_User, time=1, value=0}
          ]}}}
       Scale Curve=&Multiply_Float(A=1, B=$System.Flicker)
      )
     )
    )
   ) {isSetParameters=true, module=SetVariables_#8}
   /SetVariables(
    Particles.CoreColor:LinearColor=&MultiplyLinearColor(
     Linear Color=&LerpLinearColors(
      StartColor=<rgba 0.032 0.4515 1 0.2>
      EndColor=<rgba 0.0965 0.032 1 0.2>
      LerpFactor=&MakeFloatFromInt(Int=$User.VariantIndex)
     )
     Scale Factor=&MakeVector4_FromFloat(Float=&Multiply_Float(
       A=&FloatFromCurve(FloatCurve=DataInterface() {propertyValues {Curve {keys [
            {arriveTangent=9.7418, interpMode=RCIM_Cubic, leaveTangent=9.7418, tangentMode=RCTM_User, time=0, value=0}
            {interpMode=RCIM_Cubic, tangentMode=RCTM_User, time=0.2, value=1}
            {arriveTangent=-2.6427, interpMode=RCIM_Cubic, leaveTangent=-2.6427, tangentMode=RCTM_User, time=1, value=0}
           ]}}})
       B=&MakeFloatFromVector4(
        Vector 4=&MakeVector4_FromFloat(Float=&Lerp_Float(A=0.7, B=1, Alpha=$System.Flicker))
        Channel=::NewEnumerator0
       )
      ))
    )
    Particles.CorePosition:NiagaraPosition=&AddVectorToPosition(Position=$Particles.Initial.Position, Vector=<0 0 10>)
    Particles.CoreSize:Vector2f=&Multiply_Vector2D_ByFloat(Float=40, Vector2D=<1 1>)
   ) {isSetParameters=true, module=SetVariables_#9}
   &FN_Gen5_Light(Attenutation=100, Intensity=800, Position Offset:Vector3f=<0 0 32>)
   &FN_Light_Attributes(Color=$Particles.Color, Create Color Attribute=true, Create Exponent Attribute=true, Create Position Attribute=true, Position=<0 0 30>)
  ]
  rendererClasses [
   &NiagaraSpriteRendererProperties
   &NiagaraMeshRendererProperties
   &NiagaraComponentRendererProperties
   &NiagaraLightRendererProperties
  ]
  renderers [
   {rendererClass=&NiagaraSpriteRendererProperties}
   {rendererClass=&NiagaraMeshRendererProperties, rendererIndex=1}
   {rendererClass=&NiagaraMeshRendererProperties, rendererIndex=2}
   {rendererClass=&NiagaraMeshRendererProperties, rendererIndex=3}
   {rendererClass=&NiagaraMeshRendererProperties, rendererIndex=4}
   {rendererClass=&NiagaraComponentRendererProperties, rendererIndex=5}
   {rendererClass=&NiagaraLightRendererProperties, rendererIndex=6}
  ]
 }
 LongLightningArcs001 {
  emitterSpawnScript [
   /SetVariables(Emitter.LightningWidth:NiagaraFloat=&Multiply_Float(A=&RandomRangeFloat(Minimum=0.5, Maximum=1), B=$System.LightningWidth)) {isSetParameters=true, module=SetVariables_#4}
   /SetVariables(Emitter.LightningGlow_Width:NiagaraFloat=&Multiply_Float(A=$Emitter.LightningWidth, B=8)) {isSetParameters=true, module=SetVariables_#5}
  ]
  emitterUpdateScript [
   &EmitterState(Inactive Response=::NewEnumerator1, Life Cycle Mode=::NewEnumerator1, Loop Behavior=::NewEnumerator1, Loop Duration=2, UseLoopDelay=false)
   &SpawnBurst_Instantaneous(Spawn Count=2, Use Spawn Probability=false, UseLoopCountLimit=false) {module=SpawnBurst_Instantaneous001}
   &SpawnRate(
    Spawn Probability=0.6
    SpawnRate=&FloatFromCurve(FloatCurve=DataInterface() {propertyValues {Curve {keys [<at 0 1>,<at 0.25 0>]}}}, CurveIndex=$Emitter.Age, Scale Curve=10)
    Use Spawn Probability=true
   )
   &SpawnBurst_Instantaneous(
    Spawn Count=&MakeIntFromFloat(Float=&Lerp_Float(A=0, B=20, Alpha=$User.Debug))
    Use Spawn Probability=false
    UseLoopCountLimit=false
   )
  ]
  particleSpawnScript [
   &PartitionParticles(Number Of Partitions=4)
   &InitializeParticle(
    Color=$System.PrimaryColor
    Lifetime=&Lerp_Float(
     A=&RandomRangeFloat(Minimum=1.4, Maximum=0.5)
     B=&RandomRangeFloat(Minimum=0.4, Maximum=0.7)
     Alpha=$User.ForwardDot
    )
    Mesh Scale Mode=::NewEnumerator3
    Mesh Uniform Scale=$User.InchesConversion
    Position Offset Coordinate Space=::Local
    Sprite Size Mode=::NewEnumerator3
    Uniform Sprite Size=200
    UsePositionOffset=true
    Write Material Random=true
   )
   /SetVariables(
    Particles.ArcsOffset:NiagaraPosition=&AddVectorToPosition(Position=$Engine.Emitter.SimulationPosition, Vector=<0 0 0>)
    Particles.MeshIndex=&MakeIntFromFloat(Float=&Lerp_Float(
      A=&MakeFloatFromInt(Int=&SelectIntFromArray(
        Direct Array Index=&ModuloInt(A=&ReturnParticleIDAttributes(), B=12)
        Int Selection Array=$User.Int32 Array
       ))
      B=$User.MeshIndexOverride
      Alpha=$User.Debug
     ))
   ) {isSetParameters=true, module=SetVariables_#6}
   /SetVariables(Particles.OverlayScale:Vector3f=&VectorFromFloat(Value=2.5)) {isSetParameters=true, module=SetVariables_#7}
   &InitialMeshOrientation(
    Mesh Orientation Mode=::NewEnumerator7
    Rotation=&MakeVector(X=0, Y=0, Z=&RandomRangeFloat(Minimum=-0.4, Maximum=0.45))
    Use Rotation Vector=true
   )
   &PartitionParticles() {module=PartitionParticles001}
  ]
  particleUpdateScript [
   &ParticleState()
   &ScaleMeshSize(Scale Factor=&MakeVector(
     X=1
     Y=&RandomRangeFloat(Minimum=0.7, Maximum=1)
     Z=&FloatFromCurve(FloatCurve=DataInterface() {propertyValues {Curve {keys [{interpMode=RCIM_Cubic, tangentMode=RCTM_User, time=0, value=0}, {interpMode=RCIM_Cubic, tangentMode=RCTM_User, time=1, value=1}]}}})
    ))
   &SolveForcesAndVelocity(Clamp Velocity=false, Limit Acceleration=false)
   &ScaleColor(
    Scale RGB=&VectorFromFloat(Value=&Lerp_Float(A=0.6, B=2, Alpha=$System.Flicker))
    ScaleA=false
    ScaleRGB=true
   )
   &ScaleColor(
    Color Value To Scale=$Particles.Color
    Scale RGB=&Multiply_VectorByFloat(Vector=<1 1 1>, Float=$System.TimeOfDayBrightnessMultiplier)
    ScaleA=true
    ScaleRGB=true
   ) {module=ScaleColor001}
   &DynamicMaterialParameters(
    Index 0 Param 1=&FloatFromCurve(
     FloatCurve=DataInterface() {propertyValues {Curve {keys [
         {interpMode=RCIM_Cubic, tangentMode=RCTM_User, time=0, value=0}
         {interpMode=RCIM_Cubic, tangentMode=RCTM_User, time=0.5, value=1.05}
         {interpMode=RCIM_Cubic, tangentMode=RCTM_User, time=1, value=0}
        ]}}}
     Scale Curve=&RandomRangeFloat(Minimum=1, Maximum=2)
    )
    Index 0 Param 4=&RandomRangeFloat(Minimum=1.4, Maximum=1)
    Module.Index0Param0Write=true
    Module.Index0Param1Write=true
    Module.Index0Param2Write=true
    Module.Index0Param3Write=true
   )
   /SetVariables(
    Particles.ContactColor:LinearColor=&MultiplyLinearColor(
     Linear Color=$Particles.Color
     Scale Factor=&MakeFloat4FromVectorAndFloat(
      "Vector (RGB)"=&VectorFromFloat(Value=&FNTimeOfDayScalar(Night Scale=0.1, Morning Scale=1, Evening Scale=1))
      "Float (Alpha)"=&FloatFromCurve(FloatCurve=DataInterface() {propertyValues {Curve {keys [
           {interpMode=RCIM_Cubic, tangentMode=RCTM_User, time=0.45, value=0}
           {interpMode=RCIM_Cubic, tangentMode=RCTM_User, time=0.5, value=1}
           {interpMode=RCIM_Cubic, tangentMode=RCTM_User, time=1, value=0}
          ]}}})
     )
    )
    Particles.LightningGlowColor:LinearColor=&MultiplyLinearColor(
     Linear Color=&MultiplyLinearColor(
      Linear Color=$Particles.Color
      Scale Factor=&MakeVector4_FromLinearColor(LinearColor=&LerpLinearColors(
        StartColor=<rgba 0 0.9091 4 1>
        EndColor=<rgba 4 0 2.0667 1>
        LerpFactor=&MakeFloatFromInt(Int=$User.VariantIndex)
       ))
     )
     Scale Factor=&MakeFloat4FromVectorAndFloat("Vector (RGB)"=<0.1 0.1 0.1>, "Float (Alpha)"=0.5)
    )
    Particles.OverlayColor:LinearColor=&MultiplyLinearColor(
     Linear Color=&MultiplyLinearColor(
      Linear Color=$Particles.Color
      Scale Factor=&MakeVector4_FromLinearColor(LinearColor=&LerpLinearColors(
        StartColor=<rgba 0 2.2 4 1>
        EndColor=<rgba 2.4667 0 4 1>
        LerpFactor=&MakeFloatFromInt(Int=$User.VariantIndex)
       ))
     )
     Scale Factor=&MakeFloat4FromVectorAndFloat(
      "Vector (RGB)"=<1 1 1>
      "Float (Alpha)"=&FloatFromCurve(
       FloatCurve=DataInterface() {propertyValues {Curve {keys [
           {interpMode=RCIM_Cubic, tangentMode=RCTM_User, time=0, value=0}
           {interpMode=RCIM_Cubic, tangentMode=RCTM_User, time=0.8, value=1}
           {interpMode=RCIM_Cubic, tangentMode=RCTM_User, time=1, value=0}
          ]}}}
       Scale Curve=&Multiply_Float(A=1, B=$System.Flicker)
      )
     )
    )
   ) {isSetParameters=true, module=SetVariables_#8}
   /SetVariables(
    Particles.CoreColor:LinearColor=&MultiplyLinearColor(
     Linear Color=&LerpLinearColors(
      StartColor=<rgba 0.032 0.4515 1 0.2>
      EndColor=<rgba 0.0965 0.032 1 0.2>
      LerpFactor=&MakeFloatFromInt(Int=$User.VariantIndex)
     )
     Scale Factor=&MakeVector4_FromFloat(Float=&Multiply_Float(
       A=&FloatFromCurve(FloatCurve=DataInterface() {propertyValues {Curve {keys [
            {arriveTangent=9.7418, interpMode=RCIM_Cubic, leaveTangent=9.7418, tangentMode=RCTM_User, time=0, value=0}
            {interpMode=RCIM_Cubic, tangentMode=RCTM_User, time=0.2, value=1}
            {arriveTangent=-2.6427, interpMode=RCIM_Cubic, leaveTangent=-2.6427, tangentMode=RCTM_User, time=1, value=0}
           ]}}})
       B=&MakeFloatFromVector4(
        Vector 4=&MakeVector4_FromFloat(Float=&Lerp_Float(A=0.7, B=1, Alpha=$System.Flicker))
        Channel=::NewEnumerator0
       )
      ))
    )
    Particles.CorePosition:NiagaraPosition=&AddVectorToPosition(Position=$Particles.Initial.Position, Vector=<0 0 10>)
    Particles.CoreSize:Vector2f=&Multiply_Vector2D_ByFloat(Float=40, Vector2D=<1 1>)
   ) {isSetParameters=true, module=SetVariables_#9}
   &FN_Gen5_Light(Attenutation=100, Intensity=800, Position Offset:Vector3f=<0 0 32>)
   &FN_Light_Attributes(Color=$Particles.Color, Create Color Attribute=true, Create Exponent Attribute=true, Create Position Attribute=true, Position=<0 0 30>)
  ]
  rendererClasses [
   &NiagaraMeshRendererProperties
   &NiagaraSpriteRendererProperties
   &NiagaraComponentRendererProperties
   &NiagaraLightRendererProperties
  ]
  renderers [
   {rendererClass=&NiagaraMeshRendererProperties, rendererIndex=0}
   {rendererClass=&NiagaraMeshRendererProperties, rendererIndex=1}
   {rendererClass=&NiagaraMeshRendererProperties, rendererIndex=2}
   {rendererClass=&NiagaraSpriteRendererProperties, rendererIndex=3}
   {rendererClass=&NiagaraMeshRendererProperties, rendererIndex=4}
   {rendererClass=&NiagaraComponentRendererProperties, rendererIndex=5}
   {rendererClass=&NiagaraLightRendererProperties, rendererIndex=6}
  ]
 }
 SimpleSpriteBurst {
  emitterUpdateScript [
   &EmitterState(Life Cycle Mode=::NewEnumerator1, Loop Behavior=::NewEnumerator1, Loop Duration Mode=::NewEnumerator1, UseLoopDelay=false)
   &SpawnBurst_Instantaneous(Use Spawn Probability=false, UseLoopCountLimit=false) {enabled=false}
   &SpawnBurst_Instantaneous(Spawn Group=1, Use Spawn Probability=false, UseLoopCountLimit=false) {module=SpawnBurst_Instantaneous001}
  ]
  particleSpawnScript [
   &InitializeParticle(
    Color=$System.PrimaryColor
    Position Offset=<0 0 16>
    Sprite Size Mode=::NewEnumerator3
    Uniform Sprite Size=50
    UsePositionOffset=true
    Write Material Random=true
   )
   &CameraOffset(Camera Offset Amount=30)
  ]
  particleUpdateScript [
   &ParticleState(Kill Particles When Lifetime Has Elapsed=false)
   /SetVariables(Particles.Linear Color=&MultiplyLinearColor(
     Linear Color=<rgba 0 21.2329 40 1>
     Scale Factor=&MakeVector4_FromFloat(Float=&FloatFromCurve(FloatCurve=DataInterface() {propertyValues {Curve {keys [<at 0 0>,<at 1 1>]}}}, CurveIndex=$Emitter.Age, Scale Curve=0.1))
    )) {isSetParameters=true}
   &ScaleColor(
    Scale Alpha=&FloatFromCurve(
     FloatCurve=DataInterface() {propertyValues {Curve {keys [
         {interpMode=RCIM_Cubic, tangentMode=RCTM_User, time=0, value=0}
         {arriveTangent=0.002, interpMode=RCIM_Cubic, leaveTangent=0.002, tangentMode=RCTM_User, time=0.2, value=1.0113}
         {interpMode=RCIM_Cubic, tangentMode=RCTM_User, time=1.0011, value=0.002}
        ]}}}
     Scale Curve=&FNTimeOfDayScalar(Night Scale=0.3, Morning Scale=1, Evening Scale=1)
    )
    ScaleA=true
    ScaleRGB=false
   )
   &SolveForcesAndVelocity(Clamp Velocity=false, Limit Acceleration=false)
  ]
  rendererClasses [&NiagaraSpriteRendererProperties]
  renderers [
   {rendererClass=&NiagaraSpriteRendererProperties}
   {rendererClass=&NiagaraSpriteRendererProperties, rendererIndex=1}
   {rendererClass=&NiagaraSpriteRendererProperties, rendererIndex=2}
  ]
 }
}
summary {
 systemName=NS_Season27_Glider_Idle
 emitters [
  {
   emitterName=LongLightningArcs
   rendererClasses [
    &NiagaraSpriteRendererProperties
    &NiagaraMeshRendererProperties
    &NiagaraComponentRendererProperties
    &NiagaraLightRendererProperties
   ]
  }
  {emitterName=Arrow, rendererClasses [&NiagaraMeshRendererProperties, &NiagaraSpriteRendererProperties]}
  {emitterName=SimpleSpriteBurst, rendererClasses [&NiagaraSpriteRendererProperties]}
  {
   emitterName=LongLightningArcs001
   rendererClasses [
    &NiagaraMeshRendererProperties
    &NiagaraSpriteRendererProperties
    &NiagaraComponentRendererProperties
    &NiagaraLightRendererProperties
   ]
  }
 ]
 userVariables [
  User.bIsFrontEnd {type=NiagaraBool}
  User.bIsFrontEndPreview {type=NiagaraBool}
  User.Color Array {
   type=NiagaraDataInterfaceArrayColor
   description=""
   defaultValue {
    struct=NiagaraExt_VariableValue_DataInterface
    value {
     dataInterface=&self:NiagaraDataInterfaceArrayColor_0
     dataInterfaceClass=&NiagaraDataInterfaceArrayColor
    }
   }
  }
  User.Debug {type=NiagaraFloat, defaultValue=0, description=""}
  User.FlickerSpeed {type=NiagaraFloat, defaultValue=6, description=""}
  User.ForwardDot {type=NiagaraFloat, defaultValue=0, description=""}
  User.InchesConversion {type=NiagaraFloat, defaultValue=2.54, description=""}
  User.Int32 Array {
   type=NiagaraDataInterfaceArrayInt32
   description=""
   defaultValue {
    struct=NiagaraExt_VariableValue_DataInterface
    value {
     dataInterface=&self:NiagaraDataInterfaceArrayInt32_0
     dataInterfaceClass=&NiagaraDataInterfaceArrayInt32
    }
   }
  }
  User.MeshIndexOverride {type=NiagaraFloat, defaultValue=1, description=""}
  User.VariantIndex {type=NiagaraInt32}
 ]
}
